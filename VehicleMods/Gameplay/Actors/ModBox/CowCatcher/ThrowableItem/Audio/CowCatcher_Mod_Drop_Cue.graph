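# Sound cue for the drop sound of the CowCatcher vehicle mod's throwable item.
# A Random node picks one of four equally weighted drop wave variants and feeds a Modulator at the root to vary playback.
# It plays at reduced volume in the Fort vehicles sound class, using the shared weapon-drop attenuation.
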SoundCue /VehicleMods/Gameplay/Actors/ModBox/CowCatcher/ThrowableItem/Audio/CowCatcher_Mod_Drop_Cue
root=Modulator_0
nodes [
 {id=Modulator_0, children [Random_0]}
 {id=Random_0, children [WavePlayer_1,WavePlayer_5,WavePlayer_6,WavePlayer_7], properties {weights [1,1,1,1]}}
 {id=WavePlayer_1, properties {soundWaveAssetPtr=&/CowCatcher_Mod_Drop_04}}
 {id=WavePlayer_5, properties {soundWaveAssetPtr=&/CowCatcher_Mod_Drop_03}}
 {id=WavePlayer_6, properties {soundWaveAssetPtr=&/CowCatcher_Mod_Drop_02}}
 {id=WavePlayer_7, properties {soundWaveAssetPtr=&/CowCatcher_Mod_Drop_01}}
]
properties {
 attenuationSettings=&Fort_Weapon_Drop_Att
 maxDistance=3000
 soundClassObject=&Fort_Vehicles
 volumeMultiplier=0.25
}
