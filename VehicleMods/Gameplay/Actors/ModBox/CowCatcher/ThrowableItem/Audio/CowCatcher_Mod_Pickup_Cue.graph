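# Sound cue for picking up the throwable cow catcher vehicle mod item.
# It picks one of three pickup wave variants at equal weight and passes it through a modulator.
# Playback is quieter and slightly lower pitched, routed to the vehicles sound class with gadget-throw attenuation for distance falloff.
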SoundCue /VehicleMods/Gameplay/Actors/ModBox/CowCatcher/ThrowableItem/Audio/CowCatcher_Mod_Pickup_Cue
root=Modulator_0
nodes [
 {id=Modulator_0, children [Random_0]}
 {id=Random_0, children [WavePlayer_0,WavePlayer_4,WavePlayer_5], properties {weights [1,1,1]}}
 {id=WavePlayer_0, properties {soundWaveAssetPtr=&/CowCatcher_Mod_Pickup_03}}
 {id=WavePlayer_4, properties {soundWaveAssetPtr=&/CowCatcher_Mod_Pickup_02}}
 {id=WavePlayer_5, properties {soundWaveAssetPtr=&/CowCatcher_Mod_Pickup_01}}
]
properties {
 attenuationSettings=/Game/Sounds/Attenuation/Fort_GadgetThrow_Att
 maxDistance=1536
 pitchMultiplier=0.9
 soundClassObject=&Fort_Vehicles
 volumeMultiplier=0.3
}
